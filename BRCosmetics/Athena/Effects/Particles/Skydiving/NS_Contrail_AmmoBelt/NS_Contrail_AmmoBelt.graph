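# Skydiving contrail cosmetic that trails ammo belts from the player's hands and feet and sheds tumbling bullets.
# FN_Contrails_Kite_v2 reuses kite-string modules across four bones for mesh and ribbon trails, scaled separately for hands and feet and driven by dive attributes.
# Custom quaternion swizzle and spin modules orient and counter-rotate its meshes.
# Hand_Bullets spawns meshes at those bones, ejecting them against velocity with gravity, spin and shrink-out.
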NiagaraSystem /BRCosmetics/Athena/Effects/Particles/Skydiving/NS_Contrail_AmmoBelt/NS_Contrail_AmmoBelt
dependencies {
 usedDataInterfaces [&NiagaraDataInterfaceCurve, &NiagaraDataInterfaceVectorCurve]
 usedDynamicInputs []
 usedModules [&SystemState]
}
emitters {
 FN_Contrails_Kite_v2 {
  emitterUpdateScript [
   &EmitterScalability(
    Module.bApplySpawnCountScale=false
    Module.bApplySpawnCountScaleByDistanceFraction=false
    Module.bUseMaxDistance=false
    Module.bUseMinDistance=false
   ) {enabled=false}
   &EmitterLifeCycle(NextLoopDuration=0.9833) {enabled=false}
   &EmitterState()
   /Game/Effects/Niagara/Modules/Spawn/FN_Spawn_Kite(
    "Link Per Kite  Amount"=1
    Noise=DataInterface() {propertyValues {
      XCurve {keys [
        <at 0 0 RCIM_Cubic>
        {arriveTangent=-22.177, interpMode=RCIM_Cubic, leaveTangent=-22.177, time=0.0148, value=0.3241}
        {arriveTangent=8.0832, interpMode=RCIM_Cubic, leaveTangent=8.0832, time=0.0653, value=-1.4471}
        {arriveTangent=39.4964, interpMode=RCIM_Cubic, leaveTangent=39.4964, time=0.1184, value=1.1614}
        <at 0.1293 1.083>
        {arriveTangent=-23.4377, interpMode=RCIM_Cubic, leaveTangent=-23.4377, time=0.1576, value=0.0233}
        {arriveTangent=-0.0218, interpMode=RCIM_Cubic, leaveTangent=-0.0218, time=0.2029, value=-0.642}
        <at 0.267 0.0209>
        {arriveTangent=-9.845, interpMode=RCIM_Cubic, leaveTangent=-9.845, time=0.2888, value=0.1309}
        {arriveTangent=3.938, interpMode=RCIM_Cubic, leaveTangent=3.938, time=0.3474, value=-0.7708}
        {arriveTangent=2.5316, interpMode=RCIM_Cubic, leaveTangent=2.5316, time=0.3869, value=0.5173}
        {arriveTangent=-4.044, interpMode=RCIM_Cubic, leaveTangent=-4.044, time=0.4237, value=-0.5776}
        <at 0.4469 0.2748>
        {arriveTangent=-4.0955, interpMode=RCIM_Cubic, leaveTangent=-4.0955, time=0.4551, value=0.4851}
        {arriveTangent=-8.4961, interpMode=RCIM_Cubic, leaveTangent=-8.4961, time=0.5107, value=0.0134}
        <at 0.5109 0.0104>
        {arriveTangent=-0.6728, interpMode=RCIM_Cubic, leaveTangent=-0.6728, time=0.5736, value=-1.125}
        <at 0.6118 -0.0575>
        {arriveTangent=17.1461, interpMode=RCIM_Cubic, leaveTangent=17.1461, time=0.6146, value=0.0105}
        {arriveTangent=11.3424, interpMode=RCIM_Cubic, leaveTangent=11.3424, time=0.6622, value=0.8072}
        {arriveTangent=-13.2445, interpMode=RCIM_Cubic, leaveTangent=-13.2445, time=0.699, value=0.9682}
        {arriveTangent=-25.0454, interpMode=RCIM_Cubic, leaveTangent=-25.0454, time=0.7226, value=0.0074}
        {arriveTangent=-0.028, interpMode=RCIM_Cubic, leaveTangent=-0.028, time=0.7672, value=-0.7386}
        {arriveTangent=21.9983, interpMode=RCIM_Cubic, leaveTangent=21.9983, time=0.7974, value=0.0053}
        {arriveTangent=5.888, interpMode=RCIM_Cubic, leaveTangent=5.888, time=0.8462, value=1.0004}
        {arriveTangent=-15.155, interpMode=RCIM_Cubic, leaveTangent=-15.155, time=0.8953, value=0.5818}
        <at 0.9119 0.0053>
        {arriveTangent=-7.8259, interpMode=RCIM_Cubic, leaveTangent=-7.8259, time=0.9321, value=-0.6098}
        <at 0.9591 -0.3645>
        {arriveTangent=8.9177, interpMode=RCIM_Cubic, leaveTangent=8.9177, time=0.9879, value=0}
        {arriveTangent=0.682, interpMode=RCIM_Cubic, leaveTangent=0.682, tangentMode=RCTM_User, time=1, value=0}
       ]}
      YCurve {keys [
        <at 0 0 RCIM_Cubic>
        {arriveTangent=0.0179, interpMode=RCIM_Cubic, leaveTangent=0.0179, time=0.0843, value=1.1292}
        {arriveTangent=-19.1001, interpMode=RCIM_Cubic, leaveTangent=-19.1001, time=0.117, value=0.0021}
        <at 0.1293 0.2701>
        {arriveTangent=-34.1213, interpMode=RCIM_Cubic, leaveTangent=-34.1213, time=0.1607, value=1.1225}
        {arriveTangent=-41.5782, interpMode=RCIM_Cubic, leaveTangent=-41.5782, time=0.1702, value=-1.125}
        {arriveTangent=2.3241, interpMode=RCIM_Cubic, leaveTangent=2.3241, time=0.2015, value=-0.5776}
        {arriveTangent=-2.7223, interpMode=RCIM_Cubic, leaveTangent=-2.7223, time=0.2533, value=-0.9318}
        <at 0.267 -0.7557>
        {arriveTangent=19.4781, interpMode=RCIM_Cubic, leaveTangent=19.4781, time=0.3065, value=0.1309}
        {arriveTangent=-8.6722, interpMode=RCIM_Cubic, leaveTangent=-8.6722, time=0.3637, value=1.1292}
        <at 0.4469 -1.0865>
        {arriveTangent=41.2205, interpMode=RCIM_Cubic, leaveTangent=41.2205, time=0.4687, value=-1.3504}
        {arriveTangent=49.9749, interpMode=RCIM_Cubic, leaveTangent=49.9749, time=0.4959, value=0.936}
        <at 0.5109 0.7611>
        {arriveTangent=3.1016, interpMode=RCIM_Cubic, leaveTangent=3.1016, time=0.5532, value=-0.5454}
        {arriveTangent=9.5865, interpMode=RCIM_Cubic, leaveTangent=9.5865, time=0.5777, value=0.9682}
        <at 0.6118 0.0165>
        {arriveTangent=11.842, interpMode=RCIM_Cubic, leaveTangent=11.842, time=0.6309, value=-0.4165}
        {arriveTangent=-2.9535, interpMode=RCIM_Cubic, leaveTangent=-2.9535, time=0.6731, value=0.7428}
        {arriveTangent=-5.1686, interpMode=RCIM_Cubic, leaveTangent=-5.1686, time=0.6854, value=-0.5776}
        {arriveTangent=7.3333, interpMode=RCIM_Cubic, leaveTangent=7.3333, time=0.7167, value=0.5173}
        {arriveTangent=-15.9313, interpMode=RCIM_Cubic, leaveTangent=-15.9313, time=0.7641, value=0}
        <at 0.8095 -0.9606>
        {interpMode=RCIM_Cubic, tangentMode=RCTM_User, time=0.8489, value=0}
        {interpMode=RCIM_Cubic, tangentMode=RCTM_User, time=0.904, value=0.6415}
        {interpMode=RCIM_Cubic, tangentMode=RCTM_User, time=0.9329, value=0.2589}
        {interpMode=RCIM_Cubic, tangentMode=RCTM_User, time=0.967, value=-0.5063}
        {interpMode=RCIM_Cubic, tangentMode=RCTM_User, time=1, value=0}
       ]}
      ZCurve {keys [
        {interpMode=RCIM_Cubic, tangentMode=RCTM_User, time=0, value=0}
        {arriveTangent=10.357, interpMode=RCIM_Cubic, leaveTangent=10.357, tangentMode=RCTM_User, time=0.0489, value=-1.0284}
        {arriveTangent=5.6189, interpMode=RCIM_Cubic, leaveTangent=5.6189, tangentMode=RCTM_User, time=0.0854, value=0.8726}
        {arriveTangent=-0.1676, interpMode=RCIM_Cubic, leaveTangent=-0.1676, tangentMode=RCTM_User, time=0.1245, value=-0.6036}
        {arriveTangent=-0.5858, interpMode=RCIM_Cubic, leaveTangent=-0.5858, tangentMode=RCTM_User, time=0.1956, value=0.8542}
        {interpMode=RCIM_Cubic, tangentMode=RCTM_User, time=0.219, value=-0.659}
        {interpMode=RCIM_Cubic, tangentMode=RCTM_User, time=0.2776, value=0.8542}
        {arriveTangent=-3.0163, interpMode=RCIM_Cubic, leaveTangent=-3.0163, tangentMode=RCTM_User, time=0.2916, value=-0.659}
        {arriveTangent=7.8759, interpMode=RCIM_Cubic, leaveTangent=7.8759, tangentMode=RCTM_User, time=0.3877, value=0.522}
        {arriveTangent=-10.0071, interpMode=RCIM_Cubic, leaveTangent=-10.0071, tangentMode=RCTM_User, time=0.4439, value=0.5405}
        {arriveTangent=-1.5965, interpMode=RCIM_Cubic, leaveTangent=-1.5965, tangentMode=RCTM_User, time=0.5205, value=-0.8066}
        {arriveTangent=1.6215, interpMode=RCIM_Cubic, leaveTangent=1.6215, tangentMode=RCTM_User, time=0.5595, value=0.3559}
        {arriveTangent=2.5933, interpMode=RCIM_Cubic, leaveTangent=2.5933, tangentMode=RCTM_User, time=0.6001, value=-0.6774}
        {arriveTangent=26.7367, interpMode=RCIM_Cubic, leaveTangent=26.7367, tangentMode=RCTM_User, time=0.6235, value=0.522}
        {arriveTangent=-16.0548, interpMode=RCIM_Cubic, leaveTangent=-16.0548, tangentMode=RCTM_User, time=0.6595, value=0.9095}
        {arriveTangent=-7.1505, interpMode=RCIM_Cubic, leaveTangent=-7.1505, tangentMode=RCTM_User, time=0.6845, value=-0.456}
        {arriveTangent=5.7226, interpMode=RCIM_Cubic, leaveTangent=5.7226, tangentMode=RCTM_User, time=0.7188, value=0.4851}
        {interpMode=RCIM_Cubic, tangentMode=RCTM_User, time=0.7641, value=0}
        <at 0.8148 -0.6258>
        {arriveTangent=6.4436, interpMode=RCIM_Cubic, leaveTangent=6.4436, time=0.8699, value=0.7849}
        {interpMode=RCIM_Cubic, tangentMode=RCTM_User, time=0.9119, value=0}
        <at 0.9565 -0.6736 RCIM_Cubic>
        <at 1 0>
       ]}
     }}
    Number of Bones=4
   )
  ]
  particleSpawnScript [
   /Game/Effects/Niagara/Modules/Kite/FN_Initialize_Kite_v2(Module.Write Color=true, Module.Write RibbonTwist=false, Module.Write RibbonWidth=true)
   /SetVariables(
    Particles.FeetScale:Vector3f=&MakeCustomVectorFromBool(Bool=HlslExpression() {hlslExpression="Particles.SpawnGroup < 2"}, True Vector=<0 0 0>, False Vector=<1 1 1>)
    Particles.HandsScale:Vector3f=&MakeCustomVectorFromBool(Bool=HlslExpression() {hlslExpression="Particles.SpawnGroup < 2"}, True Vector=<3 3 2.2>, False Vector=<0 0 0>)
   ) {isSetParameters=true}
  ]
  particleUpdateScript [
   /Game/Effects/Niagara/Modules/Skydiving/FN_DivingAttributes/FN_DivingAttributes() {enabled=false}
   /Game/Effects/Niagara/Modules/Skydiving/FN_DivingAttributes/FN_DivingAttributes_v2(SkeletalMesh=DataInterface() {propertyValues {
      DefaultMesh=null
      ExcludeBoneName=root
      OptionalComponentClass=null
      PreviewMesh=&M_Med_Soldier_04_Preview
      SkinningMode=SkinOnTheFly
      SoftSourceActor=null
      SourceMode=Default
      UvSetIndex=0
      WholeMeshLOD=-1
      bExcludeBone=false
      bReadDeformedGeometry=true
      bRequireCurrentFrameData=true
      ComponentTags []
      FilteredBones []
      FilteredSockets []
      MeshUserParameter {parameter {name=None, typeDefHandle {registeredTypeIndex=1}, varData []}}
      SamplingRegions []
     }})
   /Game/Effects/Niagara/Modules/Kite/FN_Update_Kite_v2(
    Amplitude=0
    Frequency=0
    Kite Position Offset:Vector3f=<0 0 1>
    Kite Scale:Vector3f=&VectorFromFloat(Value=3)
    Rotation Amplitude=0
    Scale Kite Position=0.4545
    Scale String Start Pos=1
    Spool Position Offset:Vector3f=/BRCosmetics/Effects/Niagara/DynamicInputs/Skydiving/SkydiveHandsFeetVector3Modifier(Hands:Vector3f=<20 0 0>, Feet:Vector3f=<0 0 0>)
    Spool Scale:Vector3f=&VectorFromFloat(Value=/Game/Effects/Niagara/DynamicInputs/Skydiving/SkydiveHandsFeetFloatModifier(Hands:NiagaraFloat=1.5, Feet:NiagaraFloat=1.3))
    String End Offset:Vector3f=<0 0 0>
    String Length Limit=30
   )
   &KillParticles(Kill Particles=/Game/Effects/Niagara/DynamicInputs/State/FN_IsExecutionStateActive()) {enabled=false}
   /Game/Effects/Niagara/Modules/Materials/FN_MeshMaterial(Amplitude=100, Falloff=500, Frequency=200, Offset:NiagaraFloat=$Particles.AccumulatedNormalizedDive, Scale Noise X Axis=0.21, Speed=0.75)
   /BRCosmetics/Athena/Effects/Particles/Skydiving/NS_Contrail_AmmoBelt/Module/FN_QuaterionSwizzle(
    From Vector:Vector3f=&MakeVector(X=0, Y=HlslExpression() {hlslExpression="(Particles.UniqueID && 1) ? -1.0f : 1.0f"}, Z=0)
    Quat:Quat4f=$Particles.SampledRotation
    Set Mesh Orientation=true
    To Vector:Vector3f=<1 0 0>
   )
   /BRCosmetics/Athena/Effects/Particles/Skydiving/NS_Contrail_AmmoBelt/Module/FN_QuaternionSpin(
    Angle:NiagaraFloat=&Multiply_Float(
     A=$Emitter.Age
     B=&Multiply_Float(A=HlslExpression() {hlslExpression="(Particles.UniqueID && 1 ^ ((Particles.SpawnGroup < 2) ? 1 : 0)) ? -1.0f : 1.0f"}, B=1.5)
    )
    Quaternion:Quat4f=$Particles.FN_QuaterionSwizzle.Quat
    Set Mesh Orientation=true
   )
   /SetVariables(
    Particles.DynamicMaterialParameter:Vector4f=&MakeVector4(X=&MakeFloatFromInt(Int=$Particles.UniqueID), Y=0, Z=0, W=0)
    Particles.RenderPosition:Vector3f=&Add_Vector(A=$Particles.Spool.Position, B=<50 0 0>)
   ) {enabled=false, isSetParameters=true}
  ]
  rendererClasses [&NiagaraMeshRendererProperties, &NiagaraRibbonRendererProperties]
  renderers [
   {rendererClass=&NiagaraMeshRendererProperties, rendererIndex=0}
   {rendererClass=&NiagaraMeshRendererProperties, rendererIndex=1}
   {rendererClass=&NiagaraMeshRendererProperties, rendererIndex=2}
   {rendererClass=&NiagaraRibbonRendererProperties, rendererIndex=3}
  ]
 }
 Hand_Bullets {
  emitterSpawnScript [&FN_PreserveSkydiveUserVariables() {module=PreserveUserVariables}]
  emitterUpdateScript [
   &EmitterState()
   &SpawnRate(SpawnRate=10, Use Spawn Probability=false)
  ]
  particleSpawnScript [
   /Game/Effects/Niagara/Modules/Groups/FN_SetSpawnGroup_v1/FN_SetSpawnGroup(Emitter Bone Count=4) {enabled=false}
   &FN_SetSpawnGroup_v2(Number of Spawn Groups=4) {enabled=false}
   /Game/Effects/Niagara/Modules/Groups/FN_SetSpawnGroup() {module=FN_SetSpawnGroup001}
   /Niagara/Modules/Spawn/Initialization/InitializeParticle(
    Lifetime=&UniformRangedFloat(Minimum=0.6, Maximum=1.2)
    Module.Write Color=true
    Module.Write Lifetime=true
    Module.Write Mass=true
    Module.Write Material Random=true
    Module.Write Position=false
    Module.Write Scale=true
    Module.Write SpriteRotation=false
    Module.Write SpriteSize=false
    Position=&SimulationPosition()
   )
   &SampleSkeletalMeshSkeleton(Mesh:NiagaraDataInterfaceSkeletalMesh=$User.SkeletalMesh, Specific Bone Index:NiagaraInt32=$Particles.SpawnGroup) {enabled=false}
   /Niagara/Modules/Spawn/Location/SkeletalMeshSkeletonLocation(Module.Inherit Velocity=false, Module.Use Sampled Bone as Ribbon ID=true) {enabled=false}
   &SkeletalMeshLocation(
    Bone / Socket Index=$Particles.SpawnGroup
    Bone Sampling Mode=::NewEnumerator4
    Inherit Velocity=false
    MaskSpawnGroup=false
    Skeletal Mesh=$User.SkeletalMesh
   )
   /BRCosmetics/Athena/Effects/Particles/Skydiving/NS_Contrail_AmmoBelt/Module/FN_OrientMeshToEngineOwner(Set Mesh Orientation=true)
   /SetVariables(Particles.Position=&Add_Vector(
     A=Linked() {linkedVariable {name=Particles.Position, type=NiagaraPosition}}
     B=&Multiply_VectorByFloat(
      Vector=/Niagara/DynamicInputs/Helpers/SafeNormalizeVector(Vector To Normalize:Vector3f=$Particles.Velocity, Fallback Vector:Vector3f=<0 0 1>)
      Float=HlslExpression() {hlslExpression="Particles.SpawnGroup < 2 ?  95: 60"}
     )
    )) {isSetParameters=true}
   &AddVelocityInCone(
    Cone Angle=10
    Cone Axis=&SafeNormalizeVector(
     Vector To Normalize:Vector3f=&Multiply_VectorByFloat(Vector=$Particles.Velocity, Float=-1)
     Fallback Vector:Vector3f=<0 0 1>
    )
    Cone Axis Coordinate Space=::World
    Module.Use Velocity Falloff On Cone Axis=false
    Velocity Strength=&UniformRangedFloat(Minimum=100, Maximum=300)
   )
   &MeshRotationForce(
    "Lever Radius (cm)":NiagaraFloat=&UniformRangedFloat(Minimum=150, Maximum=350)
    Rotation=/Niagara/DynamicInputs/Random/RandomVector(
     Vector Scale:NiagaraFloat=1
     Constraint Vector:Vector3f=<0 0 1>
     Randomness Mode:ENiagaraRandomnessMode=::NewEnumerator0
     Random Seed:NiagaraInt32=0
     Fixed Random Seed:NiagaraBool=false
    )
   )
   &ApplyInitialForces()
  ]
  particleUpdateScript [
   &ParticleState()
   &GravityForce()
   &ScaleMeshSize(Scale Factor=&VectorFromFloat(Value=&FloatFromCurve(FloatCurve=DataInterface() {propertyValues {Curve {keys [
          <at 0.0008 0.667 RCIM_Cubic>
          {interpMode=RCIM_Cubic, tangentMode=RCTM_User, time=0.1657, value=1.0109}
          {interpMode=RCIM_Cubic, tangentMode=RCTM_User, time=0.7195, value=1.0073}
          {arriveTangent=-4.4535, interpMode=RCIM_Cubic, leaveTangent=-4.4535, tangentMode=RCTM_User, time=1.0011, value=0.002}
         ]}}})))
   &SolveForcesAndVelocity(Clamp Velocity=false, Limit Acceleration=false)
   &SolveRotationalForcesAndVelocity()
  ]
  rendererClasses [&NiagaraMeshRendererProperties]
  renderers [{rendererClass=&NiagaraMeshRendererProperties, rendererIndex=0}]
 }
}
summary {
 systemName=NS_Contrail_AmmoBelt
 emitters [
  {emitterName=Hand_Bullets, rendererClasses [&NiagaraMeshRendererProperties]}
  {
   emitterName=FN_Contrails_Kite_v2
   rendererClasses [&NiagaraMeshRendererProperties, &NiagaraRibbonRendererProperties]
  }
 ]
 userVariables [
  User.bIsFrontEndPreview {type=NiagaraBool}
  User.ParaGlideLeanAlpha {type=NiagaraFloat, defaultValue=0, description=""}
  User.PlayerSpeed {type=NiagaraFloat, defaultValue=0, description=""}
  User.SkeletalMesh {
   type=NiagaraDataInterfaceSkeletalMesh
   defaultValue {
    struct=NiagaraExt_VariableValue_DataInterface
    value {dataInterface=&self:NiagaraDataInterfaceSkeletalMesh_0}
   }
  }
 ]
}
